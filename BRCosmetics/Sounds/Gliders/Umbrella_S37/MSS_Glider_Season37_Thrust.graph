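# Looping stereo MetaSound source for thrust audio on the Season 37 umbrella glider.
# It passes every input, including the Fort.Glider direction values, straight into the shared MSS_Glider_Thrust_Loop graph and outputs its stereo result.
# This asset supplies idle and forward loops plus forward, backward and sideways thrust sounds; backward and sideways loops and general thrust sounds are empty.
# A glider-open envelope-follower control bus modulates volume, alongside thrust-specific attenuation and concurrency.
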MetaSoundSource /BRCosmetics/Sounds/Gliders/Umbrella_S37/MSS_Glider_Season37_Thrust
inputs [
 Backward.Loop.Sounds {type=WaveAsset:Array}
 Backward.Modulator.PitchMax {type=Float}
 Backward.Modulator.PitchMin {type=Float}
 Backward.Modulator.VolumeMax {type=Float}
 Backward.Modulator.VolumeMin {type=Float}
 Backward.Sounds {type=WaveAsset:Array}
 Backward.Volume {type=Float}
 Backwards.ReverseOsTrigger {type=Bool}
 FadeTime {type=Float}
 Forward.Loop.Sounds {type=WaveAsset:Array}
 Forward.Modulator.PitchMax {type=Float}
 Forward.Modulator.PitchMin {type=Float}
 Forward.Modulator.VolumeMax {type=Float}
 Forward.Modulator.VolumeMin {type=Float}
 Forward.Sounds {type=WaveAsset:Array}
 Forward.Volume {type=Float}
 Forwards.ReverseOsTrigger {type=Bool}
 Glider.AnyDirection {type=Float}
 Glider.Backwards {type=Float}
 Glider.Forwards {type=Float}
 Glider.Sideways {type=Float}
 Sideways.Loop.Sounds {type=WaveAsset:Array}
 Sideways.Modulator.PitchMax {type=Float}
 Sideways.Modulator.PitchMin {type=Float}
 Sideways.Modulator.VolumeMax {type=Float}
 Sideways.Modulator.VolumeMin {type=Float}
 Sideways.ReverseOsTrigger {type=Bool}
 Sideways.Sounds {type=WaveAsset:Array}
 Sideways.Volume {type=Float}
 Thrust.Loop.Sounds {type=WaveAsset:Array}
 Thrust.Modulator.PitchMax {type=Float}
 Thrust.Modulator.PitchMin {type=Float}
 Thrust.Modulator.VolumeMax {type=Float}
 Thrust.Modulator.VolumeMin {type=Float}
 Thrust.OneShotVolume {type=Float}
 Thrust.ReverseOsTrigger {type=Bool}
 Thrust.Sounds {type=WaveAsset:Array}
 Thrust.Volume {type=Float}
 UE.Source.OnPlay {type=Trigger}
]
interfaces [Fort.Glider:1.0, UE.OutputFormat.Stereo:1.0, UE.Source:1.1]
outputs [
 "UE.OutputFormat.Stereo.Audio:0" {type=Audio}
 "UE.OutputFormat.Stereo.Audio:1" {type=Audio}
]
pages [Default {
  input_defaults {
   Backward.Modulator.PitchMax=1.2
   Backward.Modulator.PitchMin=0.8
   Backward.Modulator.VolumeMax=1
   Backward.Modulator.VolumeMin=0.2
   Backward.Volume=1
   Backwards.ReverseOsTrigger=null
   FadeTime=0.5
   Forward.Modulator.PitchMax=1.2
   Forward.Modulator.PitchMin=0.8
   Forward.Modulator.VolumeMax=1
   Forward.Modulator.VolumeMin=0.2
   Forward.Volume=1
   Forwards.ReverseOsTrigger=null
   Glider.AnyDirection=0
   Glider.Backwards=0
   Glider.Forwards=0
   Glider.Sideways=0
   Sideways.Modulator.PitchMax=1.2
   Sideways.Modulator.PitchMin=0.8
   Sideways.Modulator.VolumeMax=1
   Sideways.Modulator.VolumeMin=0.2
   Sideways.ReverseOsTrigger=null
   Sideways.Volume=1
   Thrust.Modulator.PitchMax=1.2
   Thrust.Modulator.PitchMin=0.8
   Thrust.Modulator.VolumeMax=1
   Thrust.Modulator.VolumeMin=0.2
   Thrust.OneShotVolume=0.6
   Thrust.ReverseOsTrigger=null
   Thrust.Volume=1
   UE.Source.OnPlay=null
   Backward.Loop.Sounds []
   Backward.Sounds [
    &/Glider_Season37_Thrust_01
    &/Glider_Season37_Thrust_02
    &/Glider_Season37_Thrust_03
   ]
   Forward.Loop.Sounds [&/Glider_Season37_Thrust_Fwd_LP_01]
   Forward.Sounds [
    &/Glider_Season37_FwdThrust_01
    &/Glider_Season37_FwdThrust_02
    &/Glider_Season37_FwdThrust_03
   ]
   Sideways.Loop.Sounds []
   Sideways.Sounds [
    &Glider_Season37_Thrust_01
    &Glider_Season37_Thrust_02
    &Glider_Season37_Thrust_03
   ]
   Thrust.Loop.Sounds [&/Glider_Season37_Thrust_Idle_LP_01]
   Thrust.Sounds []
  }
  nodes [MSS_Glider_Thrust_Loop_0 = /FortMetasound/Cosmetics/Glider/MSS_Glider_Thrust_Loop(
    Backward.Loop.Sounds=^Backward.Loop.Sounds
    Backward.Modulator.PitchMax=^Backward.Modulator.PitchMax
    Backward.Modulator.PitchMin=^Backward.Modulator.PitchMin
    Backward.Modulator.VolumeMax=^Backward.Modulator.VolumeMax
    Backward.Modulator.VolumeMin=^Backward.Modulator.VolumeMin
    Backward.Sounds=^Backward.Sounds
    Backward.Volume=^Backward.Volume
    Backwards.ReverseOsTrigger=^Backwards.ReverseOsTrigger
    FadeTime=^FadeTime
    Forward.Loop.Sounds=^Forward.Loop.Sounds
    Forward.Modulator.PitchMax=^Forward.Modulator.PitchMax
    Forward.Modulator.PitchMin=^Forward.Modulator.PitchMin
    Forward.Modulator.VolumeMax=^Forward.Modulator.VolumeMax
    Forward.Modulator.VolumeMin=^Forward.Modulator.VolumeMin
    Forward.Sounds=^Forward.Sounds
    Forward.Volume=^Forward.Volume
    Forwards.ReverseOsTrigger=^Forwards.ReverseOsTrigger
    Glider.AnyDirection=^Glider.AnyDirection
    Glider.Backwards=^Glider.Backwards
    Glider.Forwards=^Glider.Forwards
    Glider.Sideways=^Glider.Sideways
    Sideways.Loop.Sounds=^Sideways.Loop.Sounds
    Sideways.Modulator.PitchMax=^Sideways.Modulator.PitchMax
    Sideways.Modulator.PitchMin=^Sideways.Modulator.PitchMin
    Sideways.Modulator.VolumeMax=^Sideways.Modulator.VolumeMax
    Sideways.Modulator.VolumeMin=^Sideways.Modulator.VolumeMin
    Sideways.ReverseOsTrigger=^Sideways.ReverseOsTrigger
    Sideways.Sounds=^Sideways.Sounds
    Sideways.Volume=^Sideways.Volume
    Thrust.Loop.Sounds=^Thrust.Loop.Sounds
    Thrust.Modulator.PitchMax=^Thrust.Modulator.PitchMax
    Thrust.Modulator.PitchMin=^Thrust.Modulator.PitchMin
    Thrust.Modulator.VolumeMax=^Thrust.Modulator.VolumeMax
    Thrust.Modulator.VolumeMin=^Thrust.Modulator.VolumeMin
    Thrust.OneShotVolume=^Thrust.OneShotVolume
    Thrust.ReverseOsTrigger=^Thrust.ReverseOsTrigger
    Thrust.Sounds=^Thrust.Sounds
    Thrust.Volume=^Thrust.Volume
    UE.Source.OnPlay=^UE.Source.OnPlay
   )]
  links [
   {dst="@UE.OutputFormat.Stereo.Audio:0", src=MSS_Glider_Thrust_Loop_0.UE.OutputFormat.Stereo.Audio:0}
   {dst="@UE.OutputFormat.Stereo.Audio:1", src=MSS_Glider_Thrust_Loop_0.UE.OutputFormat.Stereo.Audio:1}
  ]
 }]
properties {
 assetImportData=&self:AssetImportData
 attenuationSettings=&Fort_Player_Glider_Thrust_Athena_Att
 audioPropertiesSheet=&APS_Cosmetic_Glider_Loop
 bLooping=true
 duration=10000
 maxDistance=12500
 numChannels=2
 outputFormat=Stereo
 sampleRate=48000
 soundClassObject=&Fort_Gliders_Land
 concurrencySet [&Athena_Glider_Thrust_Conc]
 modulationSettings {
  volumeRouting=Union
  volumeModulationDestination {bEnableModulation=true, modulators [/Game/Sounds/Modulation/EnvelopeFollowers/Volume/CB_Glider_Open_EnvFollower]}
 }
}
